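# Unlit translucent Niagara sprite material for stylized, outlined smoke in the BirchPalace emote.
# Two panning noises, masked by vertical gradient, distort the MainShape lookup, whose channels drive energy fill and outline.
# Stroke noise varies outline thickness, while a centerline modulated by fabric texture is subtracted from opacity for a brushed look.
# Dynamic parameters adjust noise strength, thickness and time offset; particle alpha, edge mask and quality-switched depth fade finish opacity.
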Material /BRCosmetics/Animation/Game/MainPlayer/Emotes/BirchPalace/FX/M_Emote_BirchPalace_Smoke
graph {
 properties {
  bUsedWithNiagaraSprites=true
  blendMode=BLEND_Translucent
  pixelDepthOffsetMode=PDOM_Legacy
  shadingModel=MSM_Unlit
  twoSided=true
  usedShadingModels=Unlit
 }
 nodes [
  n1 = VectorParameter() {parameterName=EnergyStr, defaultValue=<rgba 0 0.6667 1 1>}
  n2 = Multiply(A=@n1.RGB, B=ScalarParameter() {defaultValue=2, parameterName=Dissolve_GlowCStr})
  n3 = TextureCoordinate()
  n4 = VectorParameter() {parameterName=MainUVScaleAndOffset, defaultValue=<rgba 0.5 1.3613 0 0>}
  n5 = AppendVector(A=@n4.R, B=@n4.R)
  n6 = VectorParameter() {
   group=NoiseMask
   parameterName=NoiseMask_EdgeMask_Control
   channelNames {a="", b=Power, g=Multiply, r=Offset}
   defaultValue=<rgba 0.088 3.2292 3 1>
  }
  n7 = MaterialFunctionCall(Float3=@n6) {materialFunction=&BreakOutFloat3Components}
  n8 = Add(
   A=OneMinus(None=MaterialFunctionCall(Float2=@n3) {materialFunction=&BreakOutFloat2Components}.G)
   B=@n7.R
  )
  n9 = Power(Base=Saturate(None=Multiply(A=@n8, B=@n7.G)), Exp=@n7.B)
  n10 = VectorParameter() {group=MainNoise, parameterName=MainNoise_UVControl, channelNames {a=VPan, b=UPan, g=VScale, r=UScale}, defaultValue=<rgba 1 1 0 0.9854>}
  n11 = MaterialFunctionCall(
   Texture=TextureObjectParameter() {group=MainNoise, parameterName=MainNoise_Texture, sortPriority=0, texture=/Game/Items/Traps/Textures/T_PerlinRGB_01_M}
   UVControl_RGB=@n10
   UVControl_A=@n10
   "[Override]TimeInput"=NamedRerouteUsage()
  ) {materialFunction=&FXCommon_TextureSample}
  n12 = AppendVector(A=@n11.R, B=@n11.R)
  n13 = ScalarParameter() {group=MainNoise, parameterName=MainNoise_UseRR}
  n14 = VectorParameter() {
   group=MainNoise
   parameterName=MainNoise_NoiseControl
   sortPriority=60
   channelNames {a=VStrength, b=UStrength, g=VOffset, r=UOffset}
   defaultValue=<rgba -0.5 -0.5 0 0.5>
  }
  n15 = DynamicParameter() {parameterIndex=1, paramNames [NoiseAStr,Param2,Param3,Param4]}
  n16 = VectorParameter() {
   group=SecNoise
   parameterName=SecNoise_UVControl
   channelNames {a=VPan, b=UPan, g=VScale, r=UScale}
   defaultValue=<rgba 2.6396 1.5659 0 0.36>
  }
  n17 = MaterialFunctionCall(
   Texture=TextureObjectParameter() {group=SecNoise, parameterName=SecNoise_Texture, sortPriority=0, texture=&T_PerlinRGB_01_M}
   UVControl_RGB=@n16
   UVControl_A=@n16
   "[Override]TimeInput"=NamedRerouteUsage()
  ) {materialFunction=&FXCommon_TextureSample}
  n18 = AppendVector(A=@n17.R, B=@n17.R)
  n19 = ScalarParameter() {group=SecNoise, parameterName=SecNoise_UseRR}
  n20 = VectorParameter() {
   group=SecNoise
   parameterName=SecNoise_NoiseControl
   sortPriority=60
   channelNames {a=VStrength, b=UStrength, g=VOffset, r=UOffset}
   defaultValue=<rgba -0.5 -0.5 0 0.408>
  }
  n21 = Add(
   A=MaterialFunctionCall!(NoiseValue=LinearInterpolate(A=@n11.RG, B=@n12, Alpha=@n13), Control_RGB=@n14, Control_A=Multiply(A=@n15.NoiseAStr, B=@n14.VStrength)) {materialFunction=/Game/Effects/Niagara/MasterMaterials/Functions/Common/FXCommon_NoiseControl}.NoiseValue
   B=MaterialFunctionCall(NoiseValue=LinearInterpolate(A=@n17.RG, B=@n18, Alpha=@n19), Control_RGB=@n20, Control_A=@n20) {materialFunction=&FXCommon_NoiseControl}.NoiseValue
  )
  n22 = Add(
   A=Add(
    A=MaterialFunctionCall(UVs=@n3, Texture Scale=@n5) {materialFunction=&ScaleUVsByCenter}.UVs
    B=AppendVector(A=@n4.B, B=@n4.B)
   )
   B=Multiply(A=Saturate(None=@n9), B=@n21)
  )
  n23 = TextureSampleParameter2D(UVs=@n22) {
   parameterName=MainShape
   samplerType=SAMPLERTYPE_Normal
   texture=/Game/Effects/Niagara/MasterMaterials/StencilGlow/Texture/T_OutlineShape_Sphere
  }
  n24 = Saturate(None=Multiply(A=Add(A=@n23.G)) {constB=0.5})
  n25 = Ceil(None=@n24)
  n26 = Constant()
  n27 = Multiply(A=@n25, B=Saturate(None=@n26))
  n28 = VectorParameter() {parameterName=EnergyOpacityRefine, defaultValue=<rgba 9 2 0.3 0.5>}
  n29 = Add(A=MaterialFunctionCall(Texture=@n24, RGB=@n28.RGB, A=@n28.RGB) {materialFunction=&MF_FX_Refine}.Result)
  n30 = Multiply(A=@n1.RGB, B=DynamicParameter() {defaultValue=<rgba 1 1 0 1>, paramNames [Dissolve, Random, ThicknessReduce, Param4]}.Param4)
  n31 = VectorParameter() {parameterName=OutlineColor, defaultValue=<rgba 0.04 0.1467 0.2 1>}
  n32 = Saturate(None=@n26)
  n33 = LinearInterpolate(A=Add(A=Multiply(A=@n2, B=@n27), B=Multiply(A=@n29, B=@n30)), B=@n31.RGB, Alpha=@n32) {constB=0}
  n34 = Saturate(None=Multiply(A=Add(A=@n23.R)) {constB=0.5})
  n35 = TextureObjectParameter() {
   group=StrokeSize
   parameterName=StrokeSize_Texture
   samplerType=SAMPLERTYPE_LinearGrayscale
   sortPriority=0
   texture=&T_SimplePerlin_Greyscale
  }
  n36 = VectorParameter() {group=StrokeSize, parameterName=StrokeSize_UVControl, channelNames {a=VPan, b=UPan, g=VScale, r=UScale}, defaultValue=<rgba 0.5 0.5 0 0.544>}
  n37 = ScalarParameter() {defaultValue=0.2, parameterName=NoiseForStrokeStr}
  n38 = MaterialFunctionCall(
   Texture=@n35
   UVControl_RGB=@n36
   UVControl_A=@n36
   "[Override]TimeInput"=NamedRerouteUsage()
   "[Override]UVInput"=Add(A=@n3, B=Multiply(A=@n21, B=@n37))
  ) {materialFunction=&FXCommon_TextureSample}
  n39 = Add(A=@n38.R, B=ScalarParameter() {group=StrokeSize, parameterName=StrokeSize_GrayAdjust_Add})
  n40 = ScalarParameter() {defaultValue=1.5, group=StrokeSize, parameterName=StrokeSize_GrayAdjust_Mul}
  n41 = Power(
   Base=Saturate(None=Multiply(A=@n39, B=@n40))
   Exp=ScalarParameter() {defaultValue=5, group=StrokeSize, parameterName=StrokeSize_GrayAdjust_Power}
  )
  n42 = LinearInterpolate(
   A=LinearInterpolate(
    A=ScalarParameter() {defaultValue=0.9, parameterName=OutlineSize_Max}
    B=ScalarParameter() {defaultValue=0.7, parameterName=OutlineSize_Min}
    Alpha=Saturate(None=@n41)
   )
   Alpha=DynamicParameter() {defaultValue=<rgba 1 1 0 1>, paramNames [Dissolve, Random, ThicknessReduce, Param4]}.ThicknessReduce
  )
  n43 = Saturate(None=Subtract(A=@n34, B=@n42))
  n44 = OneMinus(None=@n42)
  n45 = ScalarParameter() {defaultValue=14.3563, parameterName=OutlineHardness}
  n46 = Saturate(None=Multiply(A=Divide(A=@n43, B=@n44), B=Divide(A=@n45, B=@n42)))
  n47 = Ceil(None=@n46)
  MaterialFunctionCall(Emissive=Divide(A=LinearInterpolate(A=@n33, B=@n31.RGB, Alpha=@n47) {constB=0})) {materialFunction=&MF_FX_EmissiveAdjustments}
  Multiply(A=@n21, B=ScalarParameter() {defaultValue=0.1, parameterName=NoiseForDissovleStr})
  n48 = Constant() {r=1}
  n49 = VectorParameter() {parameterName=EnergyOpacityRefine, defaultValue=<rgba 9 2 0.3 0.5>}
  n50 = ScalarParameter() {defaultValue=0.1, parameterName=BaseOpacity}
  n51 = Saturate(None=Add(
    A=MaterialFunctionCall(Texture=@n24, RGB=@n49.RGB, A=@n49.RGB) {materialFunction=&MF_FX_Refine}.Result
    B=Multiply(A=@n25, B=@n50) {constB=0.1}
   ) {constB=0.1})
  n52 = Saturate(None=Multiply(
    A=MaterialFunctionCall!(TexCoord=@n22, ClampChannel=@n48) {materialFunction=/Game/Effects/Fort_Effects/MaterialFunctions/MF_FX_TilingMethod}.Result
    B=Add(A=@n46, B=@n51)
   ))
  n53 = Subtract(A=@n34, B=@n42)
  n54 = Saturate(None=Divide(A=Saturate(None=@n53), B=OneMinus(None=@n42)))
  n55 = If(A=@n54, "A > B"=OneMinus(None=@n54), "A == B"=@n54, "A < B"=@n54) {constB=0.5}
  n56 = Saturate(None=Multiply(A=Subtract(A=@n55, B=@n24)) {constB=2})
  n57 = ScalarParameter() {group=CentlineStrokeKize, parameterName=CentlineStrokeKize_GrayAdjust_Add}
  n58 = Saturate(None=Multiply(A=Add(A=@n38.R, B=@n57), B=ScalarParameter() {defaultValue=1.5, group=CentlineStrokeKize, parameterName=CentlineStrokeKize_GrayAdjust_Mul}))
  n59 = ScalarParameter() {defaultValue=15, group=CentlineStrokeKize, parameterName=CentlineStrokeKize_GrayAdjust_Power}
  n60 = LinearInterpolate(A=Constant() {r=1}, B=ScalarParameter() {defaultValue=0.4, parameterName=Centerline_Size}, Alpha=Saturate(None=Power(Base=@n58, Exp=@n59)))
  n61 = Divide(A=Saturate(None=Subtract(A=@n56, B=@n60)), B=OneMinus(None=@n60))
  n62 = Divide(A=ScalarParameter() {defaultValue=2, parameterName=CenterLineHardness}, B=@n42)
  n63 = TextureObjectParameter() {
   group=StrokeTexture
   parameterName=StrokeTexture_Texture
   samplerType=SAMPLERTYPE_LinearColor
   sortPriority=0
   texture=/Game/Global/Textures/Default/Noise/T_FN_FabricTile_B
  }
  n64 = VectorParameter() {group=StrokeTexture, parameterName=StrokeTexture_UVControl, channelNames {a=VPan, b=UPan, g=VScale, r=UScale}, defaultValue=<rgba 3 1 0 0.5>}
  n65 = MaterialFunctionCall(
   Texture=@n63
   UVControl_RGB=@n64
   UVControl_A=@n64
   "[Override]TimeInput"=NamedRerouteUsage()
   "[Override]UVInput"=Rotator(Coordinate=TextureCoordinate(), Time=Constant() {r=3})
  ) {materialFunction=&FXCommon_TextureSample}
  n66 = Add(A=AppendVector(A=@n65.R, B=@n65.R), B=ScalarParameter() {defaultValue=0.04, group=StrokeTexture, parameterName=StrokeTexture_GrayAdjust_Add})
  n67 = ScalarParameter() {defaultValue=1.5, group=StrokeTexture, parameterName=StrokeTexture_GrayAdjust_Mul}
  n68 = Power(
   Base=Saturate(None=Multiply(A=@n66, B=@n67))
   Exp=ScalarParameter() {defaultValue=5, group=StrokeTexture, parameterName=StrokeTexture_GrayAdjust_Power}
  )
  n69 = Multiply(A=Saturate(None=Multiply(A=@n61, B=@n62)), B=OneMinus(None=Saturate(None=@n68)))
  n70 = Add(A=Multiply(A=@n25, B=@n32), B=Add(A=@n27, B=Subtract(A=@n52, B=@n69)))
  n71 = Constant() {r=1}
  n72 = TextureCoordinate()
  n73 = VectorParameter() {group=EdgeM, parameterName=EdgeM_EdgeMask_Control, channelNames {a="", b=Power, g=Multiply, r=Offset}, defaultValue=<rgba 0 2 1 1>}
  n74 = MaterialFunctionCall(Float3=@n73) {materialFunction=&BreakOutFloat3Components}
  n75 = Add(
   A=OneMinus(None=MaterialFunctionCall(Float2=@n72) {materialFunction=&BreakOutFloat2Components}.G)
   B=@n74.R
  )
  n76 = Power(Base=Saturate(None=Multiply(A=@n75, B=@n74.G)), Exp=@n74.B)
  n77 = DepthFade(Opacity=@n76, FadeDistance=ScalarParameter() {parameterName=DepthFade})
  Multiply(A=Multiply(A=Multiply(A=@n70, B=@n71), B=ParticleColor().A), B=QualitySwitch(Default=@n77, Low=@n76, High=@n77, Medium=@n76, Epic=@n77))
  NamedRerouteDeclaration(None=Add(
    A=Time() {bOverride_Period=true, period=23333}
    B=DynamicParameter() {defaultValue=<rgba 1 1 0 1>, paramNames [Dissolve, Random, ThicknessReduce, Param4]}.Random
   )) {name=TimeR, nodeColor=<rgba 0.5 0 1 1>}
  OneMinus(None=@n38.R)
 ]
}
