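# Translucent unlit Niagara sprite and mesh particle material for the BirchPalace emote that adds chromatic aberration to the scene behind it.
# Three scene color samples supply red, green and blue separately, each offset by summed vector and dynamic parameter channel values scaled by intensity.
# An exponential radial gradient fades the offsets and also drives a second quality-switched output.
# Low and Medium quality, and feature levels below SM5, substitute constants for the scene sampling.
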Material /BRCosmetics/Animation/Game/MainPlayer/Emotes/BirchPalace/FX/M_Scene_Chromatic
graph {
 properties {
  bAutomaticallySetUsageInEditor=false
  bUsedWithNiagaraMeshParticles=true
  bUsedWithNiagaraSprites=true
  blendMode=BLEND_Translucent
  shadingModel=MSM_Unlit
  translucencyPass=MTP_BeforeDOF
  usedShadingModels=Unlit
 }
 nodes [
  n1 = VectorParameter() {parameterName=Chromatic Control, channelNames {a=Intensity, b="", g="", r=""}}
  n2 = DynamicParameter() {defaultValue=<rgba 0 0 0 1>, paramNames [Chromatic Intensity, X, Y, Z]}
  n3 = Add(A=Multiply(A=@n1.Intensity) {constB=0.01}, B=Multiply(A=@n2.Chromatic Intensity) {constB=0.01})
  n4 = MaterialFunctionCall() {materialFunction=&RadialGradientExponential}
  n5 = LinearInterpolate(B=Multiply(A=Add(A=@n1.R, B=@n2.X), B=@n3), Alpha=@n4.RadialGradientExponential)
  n6 = LinearInterpolate(B=Multiply(A=Add(A=@n2.Y, B=@n1.G), B=@n3), Alpha=@n4.RadialGradientExponential)
  n7 = LinearInterpolate(B=Multiply(A=Add(A=@n2.Z, B=@n1.B), B=@n3), Alpha=@n4.RadialGradientExponential)
  n8 = MaterialFunctionCall(
   R=ComponentMask(None=SceneColor!(OffsetFraction=@n5) {inputMode=OffsetFraction, constInput=<0.01 0.01>}.RGB) {r=true}
   G=ComponentMask(None=SceneColor(OffsetFraction=@n6) {inputMode=OffsetFraction}.RGB) {g=true}
   B=ComponentMask(None=SceneColor(OffsetFraction=@n7) {inputMode=OffsetFraction, constInput=<-0.01 -0.01>}.RGB) {b=true}
  ) {materialFunction=&AppendMany}
  n9 = Constant()
  FeatureLevelSwitch(
   Default=QualitySwitch(Default=@n8.RGB, Low=@n9, High=@n8.RGB, Medium=@n9, Epic=@n8.RGB)
   ES2=@n9
   ES3_1=@n9
   SM4_REMOVED=@n9
   SM5=@n8.RGB
   SM6=@n8.RGB
  )
  n10 = Constant()
  QualitySwitch(Default=@n4.RadialGradientExponential, Low=@n10, High=@n4.RadialGradientExponential, Medium=@n10, Epic=@n4.RadialGradientExponential)
 ]
}
