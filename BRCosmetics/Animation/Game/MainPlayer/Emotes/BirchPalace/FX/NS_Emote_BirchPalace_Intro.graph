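# Niagara intro effect for the BirchPalace emote: a smoky buildup that ends in an expansion wave, timed by User.ExplosionDelay.
# MistSmoke and Chromatic sprites last just past that delay, while Pull_In_Wave starts shortly before it and ExpansionWave fires on it.
# Spinning Airflow meshes spawn on a ring, and Detail_Smoke follows socket pairs or fills a cylinder.
# Camera-angle dot products fade mist layers and squash the expansion wave when viewed along the vertical axis.
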NiagaraSystem /BRCosmetics/Animation/Game/MainPlayer/Emotes/BirchPalace/FX/NS_Emote_BirchPalace_Intro
dependencies {
 usedDataInterfaces [
  &NiagaraDataInterfaceCurve
  &NiagaraDataInterfaceVector2DCurve
  &NiagaraDataInterfaceCamera
  &NiagaraDataInterfaceColorCurve
  &NiagaraDataInterfaceVectorCurve
  &NiagaraDataInterfaceSpriteRendererInfo
  &NiagaraDataInterfaceSocketReader
 ]
 usedDynamicInputs [&MakeVector4_FromLinearColor]
 usedModules [
  &self:SystemSpawnScript.SystemScriptSource.SystemScriptGraph.NiagaraNodeAssignment_0.SetVariables_#1
  &SystemState
 ]
}
emitters {
 Airflow {
  emitterUpdateScript [
   &EmitterState(Life Cycle Mode=::NewEnumerator1, Loop Behavior=::NewEnumerator1, UseLoopDelay=true)
   &SpawnRate(
    SpawnRate=&FloatFromCurve(
     FloatCurve=DataInterface() {propertyValues {Curve {keys [<at 0 0>,<at 0.5 10 RCIM_Constant>,<at 1.4 0 RCIM_Constant>]}}}
     CurveIndex=$Emitter.Age
    )
    Use Spawn Probability=false
   )
  ]
  particleSpawnScript [
   &InitializeParticle(
    Color=<rgba 1 1 1 0.5>
    Lifetime=&RandomRangeFloat(Minimum=2, Maximum=3)
    Mesh Scale=&Multiply_VectorByFloat(
     Vector=&RandomRangeVector(Minimum=<1 1 1>, Maximum=<2 2 3>)
     Float=&RandomRangeFloat(Minimum=0.2, Maximum=0.2)
    )
    Mesh Scale Mode=::NewEnumerator4
    Position Offset=<0 0 30>
    UsePositionOffset=true
    Write Material Random=true
   )
   &ShapeLocation(
    Apply To Particle Position=::NewEnumerator1
    Handle Radius=0
    Large Radius=20
    MaskSpawnGroup=false
    Rotation Mode=::NewEnumerator1
    Shape Primitive=::NewEnumerator3
    Torus Distribution Mode:ENiagaraTorusDistributionMode=::NewEnumerator1
    U Position=&ReturnNormalizedExecIndex()
    UseOwnerScale=true
   )
   &InitialMeshOrientation(
    Mesh Orientation Mode=::NewEnumerator7
    Rotation=&RandomRangeVector(Minimum=<0 0 0>, Maximum=<0 0 1>)
    Use Rotation Vector=true
   ) {module=InitialMeshOrientation001}
   /SetVariables(
    Particles.CustomRandom:NiagaraFloat=&RandomRangeFloat(Minimum=0, Maximum=1)
    Particles.SpawnEmitterAge:NiagaraFloat=$Emitter.Age
   ) {isSetParameters=true}
   &CameraOffset(Camera Offset Amount=20)
  ]
  particleUpdateScript [
   /SetVariables(Transient.Shared.FloatFromCurve_FloatCurve=DataInterface() {propertyValues {Curve {
       defaultValue=1
       keys [<at 0 -0.3 RCIM_Cubic>, {arriveTangent=0.6523, interpMode=RCIM_Cubic, leaveTangent=0.6523, tangentMode=RCTM_User, time=1, value=0.3}]
      }}}) {isSetParameters=true}
   &ParticleState()
   &ScaleMeshSize(Scale Factor=&Multiply_Vector(
     A=&VectorFromCurve(
      VectorCurve=DataInterface() {propertyValues {
        XCurve {keys [<at 0 1 RCIM_Cubic>,<at 1 1 RCIM_Cubic>]}
        YCurve {keys [<at 0 1 RCIM_Cubic>,<at 1 1 RCIM_Cubic>]}
        ZCurve {keys [<at 0 1>,<at 1 2>]}
       }}
      Scale Curve=&Multiply_VectorByFloat(Vector=<1 1 1>, Float=1)
     )
     B=&VectorFromCurve(
      VectorCurve=DataInterface() {propertyValues {XCurve {keys [<at 0 1>,<at 1 1>]}, YCurve {keys [<at 0 1>,<at 1 1>]}, ZCurve {keys [<at 0.2 0.2>,<at 1.5 1>]}}}
      CurveIndex=$Particles.SpawnEmitterAge
      Scale Curve=&Multiply_VectorByFloat(Vector=<1 1 1>, Float=1)
     )
    ))
   &DynamicMaterialParameters(
    Index 0 Param 1=&FloatFromCurve(
     FloatCurve=?
     Transient.Shared.FloatFromCurve_FloatCurve:NiagaraDataInterfaceCurve=DataInterface() {propertyValues {Curve {
        defaultValue=1
        keys [<at 0 -0.3 RCIM_Cubic>, {arriveTangent=0.6523, interpMode=RCIM_Cubic, leaveTangent=0.6523, tangentMode=RCTM_User, time=1, value=0.3}]
       }}}
    )
    Index 0 Param 2=&Lerp_Float(A=30, B=50, Alpha=$Particles.Initial.CustomRandom)
    Index 0 Param 3=1
    Module.Index0Param0Write=true
    Module.Index0Param1Write=true
    Module.Index0Param2Write=true
    Module.Index0Param3Write=true
   )
   &ScaleColor(
    Scale Alpha=&FloatFromCurve(
     FloatCurve=DataInterface() {propertyValues {Curve {keys [{arriveTangent=-1.5781, interpMode=RCIM_Cubic, leaveTangent=-1.5781, tangentMode=RCTM_User, time=0.5, value=1}, <at 1 0 RCIM_Cubic>]}}}
     Scale Curve=&RandomRangeFloat(Minimum=1, Maximum=3)
    )
    Scale RGB=&MakeVectorFromLinearColorRGB(Color=&MergeHelper_ColorSwitch(
      0=&ColorFromCurve(ColorCurve=DataInterface() {propertyValues {
         AlphaCurve {keys [<at 0 1>]}
         BlueCurve {keys [<at 0 2>,<at 0.7495 0.0365>]}
         GreenCurve {keys [<at 0 0.8325>,<at 0.7495 0.0078>]}
         RedCurve {keys [<at 0 0.2992>,<at 0.7495 0.0069>]}
        }})
      1=&ColorFromCurve(ColorCurve=DataInterface() {propertyValues {
         AlphaCurve {keys [<at 0 1>]}
         BlueCurve {keys [<at 0.4433 0.0365>,<at 1 2>]}
         GreenCurve {keys [<at 0.4433 0.0078>,<at 1 0.8325>]}
         RedCurve {keys [<at 0.4433 0.0069>,<at 1 0.2992>]}
        }})
      2=?
      3=?
      Index=&RandomRangeInt(Minimum=1, Maximum=1)
      Num=::NewEnumerator0
     ))
    ScaleA=true
    ScaleRGB=true
   )
   &Drag(Drag=&RandomRangeFloat(Minimum=3, Maximum=5), Use Linear Drag=true, Use Rotational Drag=true)
   &MeshRotationForce(Rotation=&RandomRangeVector(Minimum=<0 0 -0.5>, Maximum=<0 0 -1>))
   &SolveForcesAndVelocity(Clamp Velocity=false, Limit Acceleration=false)
   /SetVariables(
    Particles.DarkColor:LinearColor=&MultiplyLinearColor(Linear Color=$Particles.Color, Scale Factor=<0 0 0 1>)
    Particles.DarkCyn_1:Vector4f=&MakeVector4(
     X=&FloatFromCurve(
      FloatCurve=?
      Transient.Shared.FloatFromCurve_FloatCurve:NiagaraDataInterfaceCurve=DataInterface() {propertyValues {Curve {
         defaultValue=1
         keys [<at 0 -0.3 RCIM_Cubic>, {arriveTangent=0.6523, interpMode=RCIM_Cubic, leaveTangent=0.6523, tangentMode=RCTM_User, time=1, value=0.3}]
        }}}
     )
     Y=&Lerp_Float(A=50, B=80, Alpha=$Particles.Initial.CustomRandom)
     Z=0
     W=0
    )
   ) {isSetParameters=true}
   /SetVariables(Particles.Position=&AddVectorToPosition(Position=$Engine.Emitter.SimulationPosition, Vector=$Particles.Initial.ShapeLocation.ShapeVector)) {isSetParameters=true}
  ]
  rendererClasses [&NiagaraMeshRendererProperties]
  renderers [
   {rendererClass=&NiagaraMeshRendererProperties, rendererIndex=0}
   {rendererClass=&NiagaraMeshRendererProperties, rendererIndex=1}
  ]
 }
 Chromatic {
  emitterUpdateScript [
   &EmitterState(Inactive Response=::NewEnumerator1, Life Cycle Mode=::NewEnumerator1, Loop Behavior=::NewEnumerator1, UseLoopDelay=false)
   &SpawnBurst_Instantaneous(Use Spawn Probability=false, UseLoopCountLimit=false)
   /SetVariables(Emitter.SecNoise_Control:Vector4f=&MakeVector4(
     X=-0.5
     Y=-0.5
     Z=0
     W=&Lerp_Float(
      A=-0.2
      B=-0.616
      Alpha=&FloatFromCurve(
       FloatCurve=DataInterface() {propertyValues {Curve {
          defaultValue=1
          keys [{arriveTangent=0.7798, interpMode=RCIM_Cubic, leaveTangent=0.7798, tangentMode=RCTM_User, time=0, value=0}, <at 1.5 1 RCIM_Cubic>]
         }}}
       CurveIndex=$Emitter.Age
      )
     )
    )) {isSetParameters=true, module=SetVariables_#2}
  ]
  particleSpawnScript [
   &InitializeParticle(
    Lifetime=&Add_Float(A=$User.ExplosionDelay, B=0.2)
    Sprite Size=<250 220>
    Sprite Size Mode=::NewEnumerator4
    UsePositionOffset=false
    Write Material Random=true
   )
   &CameraOffset(Camera Offset Amount=100)
  ]
  particleUpdateScript [
   &ParticleState()
   &ScaleSpriteSize(
    Non-Uniform Curve Sprite Scale=DataInterface() {propertyValues {
      XCurve {keys [{tangentMode=RCTM_User, time=0, value=0.5}, {tangentMode=RCTM_User, time=1, value=1}]}
      YCurve {keys [{tangentMode=RCTM_User, time=0, value=0.8}, {tangentMode=RCTM_User, time=1, value=1}]}
     }}
    Scale Sprite Size Mode=::NewEnumerator3
   )
   /SetVariables(Particles.SpriteAlignment=<0 0 1>) {isSetParameters=true, module=SetVariables_#3}
   &DynamicMaterialParameters(
    Index 0 Param 1=&FloatFromCurve(
     FloatCurve=DataInterface() {propertyValues {Curve {keys [
         <at 0.2 0 RCIM_Cubic>
         {arriveTangent=0.4181, interpMode=RCIM_Cubic, leaveTangent=19.9492, tangentMode=RCTM_Break, time=0.92, value=0.5}
         <at 1 2 RCIM_Cubic>
        ]}}}
     Scale Curve=2
    )
    Index 0 Param 3=&Waveform(
     Global Input Position=&RandomRangeFloat(Minimum=0, Maximum=10)
     "Waveform [1]"=::NewEnumerator10
     "[1] Frequency"=5
     "[1] Amplitude Min/Max"=<-1 1>
     "[1] Scale"=&FloatFromCurve(FloatCurve=DataInterface() {propertyValues {Curve {keys [<at 0 0.5 RCIM_Cubic>, {arriveTangent=1.7767, interpMode=RCIM_Cubic, leaveTangent=1.7767, tangentMode=RCTM_User, time=1, value=2}]}}})
    )
    Index 0 Param 4=&Waveform(
     "Waveform [1]"=::NewEnumerator9
     "[1] Frequency"=8
     "[1] Amplitude Min/Max"=<-1 1>
     "[1] Scale"=&FloatFromCurve(FloatCurve=DataInterface() {propertyValues {Curve {keys [<at 0 0.5 RCIM_Cubic>, {arriveTangent=1.7767, interpMode=RCIM_Cubic, leaveTangent=1.7767, tangentMode=RCTM_User, time=1, value=2}]}}})
    )
    Module.Index0Float0Write=true
    Module.Index0Float1Write=true
    Module.Index0Float2Write=true
    Module.Index0Float3Write=true
   )
   &ScaleColor(
    Scale Alpha=&FloatFromCurve(FloatCurve=DataInterface() {propertyValues {Curve {keys [<at 0 0>,<at 0.2 1>,<at 1 1>]}}})
    ScaleA=true
    ScaleRGB=true
   )
  ]
  rendererClasses [&NiagaraSpriteRendererProperties]
  renderers [{rendererClass=&NiagaraSpriteRendererProperties}]
 }
 Detail_Smoke {
  emitterUpdateScript [
   &EmitterState(Life Cycle Mode=::NewEnumerator1, Loop Behavior=::NewEnumerator1, Loop Delay=1, Loop Duration=2.5, UseLoopDelay=true)
   &SpawnRate(SpawnRate=30, Use Spawn Probability=false)
  ]
  particleSpawnScript [
   &PartitionParticles(
    A:NiagaraFloat=&RandomRangeFloat(Minimum=0, Maximum=1)
    B=0.66
    Comparison Type:ENiagaraBooleanLogicOps_v2=::NewEnumerator6
    Partition Mode:ENiagara_PartitionMode_v2=::NewEnumerator4
   )
   &InitializeParticle(
    Color=&MergeHelper_ColorSwitch(
     0=<rgba 0.0205 0.0259 0.0573 1>
     1=<rgba 0.6406 0.6406 0.6406 1>
     2=?
     3=?
     Index=&RandomRangeInt(Minimum=0, Maximum=1)
     Num=::NewEnumerator0
    )
    Lifetime Max=1.5
    Lifetime Mode=::NewEnumerator1
    Sprite Rotation Mode=::NewEnumerator2
    Sprite Size=&MergeHelper_Vector2Switch(
     0=&RandomRangeVector2D(Minimum=<50 50>, Maximum=<80 80>)
     1=&RandomRangeVector2D(Minimum=<100 100>, Maximum=<150 150>)
     2=?
     3=?
     4=?
     Index=$Particles.PartitionParticles.Partition
     Num=::NewEnumerator0
    )
    Sprite Size Mode=::NewEnumerator4
    UsePositionOffset=true
    Write Material Random=true
   )
   &CameraOffset(Camera Offset Amount=20)
   &ShapeLocation(
    Apply To Particle Position=::NewEnumerator1
    Cylinder Height=140
    Cylinder Height Midpoint=0.1
    Cylinder Radius=30
    MaskSpawnGroup=false
    Override Seed=false
    Shape Primitive=::NewEnumerator1
    Surface Only=false
    UseOwnerScale=true
   )
   &SocketLocation(OverrideRandomSeed=false, SamplePosOffset=false, Socket Reader=$System.Socket Reader, VelocitySpeedLimit=false)
   &self:RandomBetweenTwoSockets(
    Alpha Lerp:NiagaraFloat=&RandomRangeFloat(Minimum=0, Maximum=1)
    Index:NiagaraInt32=&MergeHelper_IntSwitch(
     0=0
     1=1
     2=3
     3=4
     4=6
     5=8
     6=9
     7=11
     8=12
     9=14
     Index=&ModuloInt(A=&ReturnParticleIDAttributes(), B=10)
     Num=::NewEnumerator8
    )
    Socket Reader:NiagaraDataInterfaceSocketReader=$System.Socket Reader
   ) {module=RandomBetweenTwoSockets001}
   /SetVariables() {isSetParameters=true}
  ]
  particleUpdateScript [
   &ParticleState()
   &self:RandomBetweenTwoSockets(
    Alpha Lerp:NiagaraFloat=&RandomRangeFloat(Minimum=0, Maximum=1)
    Index:NiagaraInt32=&MergeHelper_IntSwitch(
     0=0
     1=1
     2=3
     3=4
     4=6
     5=8
     6=9
     7=11
     8=12
     9=14
     Index=&ModuloInt(A=&ReturnParticleIDAttributes(), B=10)
     Num=::NewEnumerator8
    )
    Socket Reader:NiagaraDataInterfaceSocketReader=$System.Socket Reader
   ) {module=RandomBetweenTwoSockets}
   &SpriteRotationRate(Rotation Rate=&MergeHelper_FloatSwitch(
     0=&RandomRangeFloat(Minimum=-50, Maximum=50)
     1=&RandomRangeFloat(Minimum=-20, Maximum=20)
     2=?
     3=?
     Index=$Particles.PartitionParticles.Partition
     Num=::NewEnumerator0
    ))
   &ScaleSpriteSize(
    Scale Sprite Size Mode=::NewEnumerator0
    Uniform Scale Factor=&FloatFromCurve(
     FloatCurve=DataInterface() {propertyValues {Curve {
        defaultValue=1
        keys [
         {arriveTangent=4.1785, interpMode=RCIM_Cubic, leaveTangent=4.1785, tangentMode=RCTM_User, time=0, value=0.5}
         {arriveTangent=0.0224, interpMode=RCIM_Cubic, leaveTangent=0.0224, tangentMode=RCTM_User, time=1, value=2}
        ]
       }}}
     Scale Curve=0.38
    )
   )
   &SubUVAnimation(
    End Frame Range Override:NiagaraInt32=&MergeHelper_IntSwitch(0=25, 1=20, 2=?, 3=?, 4=?, 5=?, Index=$Particles.Initial.PartitionParticles.Partition, Num=::NewEnumerator0)
    "Sprite Renderer "=DataInterface() {propertyValues {SpriteRenderer=&self:Detail_Smoke_0.NiagaraSpriteRendererProperties_0}}
    Start Frame Range Override=&RandomRangeInt(Minimum=0, Maximum=5)
    UseEndFrame=true
    UseStartFrame=true
   )
   &ScaleColor(
    Scale Alpha=&Multiply_Float(
     A=&RandomRangeFloat(Minimum=0.5, Maximum=1)
     B=&FloatFromCurve(
      FloatCurve=DataInterface() {propertyValues {Curve {keys [
          {arriveTangent=3.7485, interpMode=RCIM_Cubic, leaveTangent=3.7485, tangentMode=RCTM_User, time=0, value=0.35}
          {interpMode=RCIM_Cubic, tangentMode=RCTM_User, time=0.5, value=1}
          {tangentMode=RCTM_User, time=1, value=0}
         ]}}}
      Scale Curve=&FN_FrontEndFloatModifier(Gameplay=2, Front End=1.5)
     )
    )
    Scale RGB=&MakeVectorFromLinearColorRGB(Color=&ColorFromCurve(ColorCurve=DataInterface() {propertyValues {
        AlphaCurve {keys [<at 0 1 RCIM_Cubic>]}
        BlueCurve {keys [<at 0 0.2865>,<at 0.4334 0.1198>,<at 0.831 0.0323 RCIM_Cubic>]}
        GreenCurve {keys [<at 0 0.1129>,<at 0.4334 0.0463>,<at 0.831 0.0333 RCIM_Cubic>]}
        RedCurve {keys [<at 0 0.0576>,<at 0.4334 0.0337>,<at 0.831 0.0365 RCIM_Cubic>]}
       }}))
    ScaleA=true
    ScaleRGB=true
   )
   /SetVariables(
    Particles.2ndSmokeColor:LinearColor=&MultiplyLinearColor(Linear Color=$Particles.Color, Scale Factor=&MakeVector4(X=1, Y=1, Z=1, W=0.6))
    Particles.2ndSmokeSize:Vector2f=&Multiply_Vector2D_ByFloat(Float=1, Vector2D=$Particles.SpriteSize)
   ) {isSetParameters=true}
   /SetVariables(Particles.Position=&Add_Vector(
     A=&MergeHelper_VectorSwitch(
      0=&MakeCustomVectorFromBool(
       Bool=&SetBoolByFloatComparison(A=&RandomRangeFloat(Minimum=0, Maximum=1), B=0.35)
       True Vector=Linked() {linkedVariable {name=Particles.RandomBetweenTwoSockets.PositionBetweenSockets, type=NiagaraPosition}}
       False Vector=Linked() {linkedVariable {name=Particles.RandomBetweenTwoSockets001.PositionBetweenSockets, type=NiagaraPosition}}
      )
      1=Linked() {linkedVariable {name=Particles.ShapeLocation.ShapePosition, type=NiagaraPosition}}
      2=?
      3=?
      Index=$Particles.PartitionParticles.Partition
      Num=::NewEnumerator0
     )
     B=&MakeVector(
      X=0
      Y=0
      Z=&FloatFromCurve(
       FloatCurve=DataInterface() {propertyValues {Curve {keys [<at 0 0 RCIM_Cubic>,<at 1 1 RCIM_Cubic>]}}}
       Scale Curve=&RandomRangeFloat(Minimum=10, Maximum=20)
      )
     )
    )) {isSetParameters=true}
  ]
  rendererClasses [&NiagaraSpriteRendererProperties]
  renderers [
   {rendererClass=&NiagaraSpriteRendererProperties}
   {rendererClass=&NiagaraSpriteRendererProperties, rendererIndex=1}
  ]
 }
 ExpansionWave {
  emitterUpdateScript [
   &EmitterState(
    Inactive Response=::NewEnumerator1
    Life Cycle Mode=::NewEnumerator1
    Loop Behavior=::NewEnumerator1
    Loop Delay=$User.ExplosionDelay
    UseLoopDelay=true
   )
   &SpawnBurst_Instantaneous(Spawn Count=1, Use Spawn Probability=false)
   &SpawnBurst_Instantaneous(Spawn Count=1, Spawn Group=1, Use Spawn Probability=false) {module=SpawnBurst_Instantaneous001}
  ]
  particleSpawnScript [
   /SetVariables(Particles.VisibilityTag=$Emitter.SpawnGroup) {isSetParameters=true}
   &InitializeParticle(
    Color=&MergeHelper_ColorSwitch(0=<rgba 1 1 1 1>, 1=<rgba 1 1 1 2>, 2=?, 3=?, Index=$Particles.VisibilityTag, Num=::NewEnumerator0)
    Lifetime=&MergeHelper_FloatSwitch(
     0=0.7
     1=&RandomRangeFloat(Minimum=0.7, Maximum=1)
     2=0
     3=0
     4=0
     Index=$Particles.VisibilityTag
     Num=::NewEnumerator3
    )
    Mesh Scale=&MergeHelper_VectorSwitch(0=<2.5 1.3 1.6>, 1=<4.5 2.8 1>, 2=?, 3=?, Index=$Particles.VisibilityTag, Num=::NewEnumerator0)
    Mesh Scale Mode=::NewEnumerator4
    Position Offset=<0 0 100>
    UsePositionOffset=true
    Write Material Random=true
   )
   &InitialMeshOrientation(Orientation Axis=<0 0 1>, Orientation Vector=<0 0 1>, Rotation=<0 0.25 0>, Use Orientation Vector=true, Use Rotation Vector=true)
   &CameraOffset(Camera Offset Amount=&MergeHelper_FloatSwitch(0=50, 1=80, 2=?, 3=?, Index=$Particles.VisibilityTag, Num=::NewEnumerator0))
  ]
  particleUpdateScript [
   &ParticleState()
   &DynamicMaterialParameters(
    Index 0 Param 1=&FloatFromCurve(FloatCurve=DataInterface() {propertyValues {Curve {keys [
         {arriveTangent=0.0769, interpMode=RCIM_Cubic, time=0, value=-0.3577}
         {arriveTangent=0.3577, interpMode=RCIM_Cubic, leaveTangent=0.3577, time=0.3, value=-0.1}
         {interpMode=RCIM_Cubic, leaveTangent=2.9562, time=1, value=0}
        ]}}})
    Index 0 Param 2=&FloatFromCurve(FloatCurve=DataInterface() {propertyValues {Curve {keys [{arriveTangent=-0.5818, interpMode=RCIM_Cubic, leaveTangent=-0.5818, tangentMode=RCTM_User, time=0, value=0.8}, <at 1 2>]}}})
    Index 0 Param 3=&FloatFromCurve(FloatCurve=DataInterface() {propertyValues {Curve {keys [<at 0 0>,<at 0.2 0.5>,<at 1 0.5>]}}}, Scale Curve=2)
    Module.Index0Float0Write=true
    Module.Index0Float1Write=true
    Module.Index0Float2Write=true
    Module.Index0Float3Write=true
   )
   &ScaleColor(
    Scale Alpha=&FloatFromCurve(FloatCurve=DataInterface() {propertyValues {Curve {keys [{arriveTangent=-2.8629, interpMode=RCIM_Cubic, time=0.3, value=1}, <at 1 0 RCIM_Cubic>]}}})
    Scale RGB=&MakeVectorFromLinearColorRGB(Color=&MergeHelper_ColorSwitch(
      0=&ColorFromCurve(ColorCurve=DataInterface() {propertyValues {
         AlphaCurve {keys [<at 0.3128 1>]}
         BlueCurve {keys [<at 0.2974 0.0521>]}
         GreenCurve {keys [<at 0.2974 0.0273>]}
         RedCurve {keys [<at 0.2974 0.023>]}
        }})
      1=&ColorFromCurve(ColorCurve=DataInterface() {propertyValues {
         AlphaCurve {keys [<at 0.3128 1>]}
         BlueCurve {keys [<at 0.8574 0.0729>,<at 0.91 0.0625>]}
         GreenCurve {keys [<at 0.8574 0.0516>,<at 0.91 0.0242>]}
         RedCurve {keys [<at 0.8574 0.0463>,<at 0.91 0.0192>]}
        }})
      2=?
      3=?
      Index=$Particles.Initial.VisibilityTag
      Num=::NewEnumerator0
     ))
    ScaleA=true
    ScaleRGB=true
   )
   &ScaleMeshSize(Scale Factor=&Multiply_Vector(
     A=&VectorFromFloat(Value=&MergeHelper_FloatSwitch(
       0=&FloatFromCurve(FloatCurve=DataInterface() {propertyValues {Curve {keys [
            {arriveTangent=3.1433, interpMode=RCIM_Cubic, leaveTangent=3.1433, tangentMode=RCTM_User, time=0, value=0.5}
            {arriveTangent=1.5, interpMode=RCIM_Cubic, leaveTangent=1.5, time=0.3, value=1.9}
            {arriveTangent=0.5979, interpMode=RCIM_Cubic, leaveTangent=0.5979, tangentMode=RCTM_User, time=1, value=2}
           ]}}})
       1=&FloatFromCurve(FloatCurve=DataInterface() {propertyValues {Curve {keys [<at 0 0>,<at 0.1 1>,<at 1 2>]}}})
       2=?
       3=?
       Index=$Particles.Initial.VisibilityTag
       Num=::NewEnumerator0
      ))
     B=&MakeVector(
      X=&Lerp_Float(
       A=1
       B=0.65
       Alpha=&FloatFromCurve(
        FloatCurve=DataInterface() {propertyValues {Curve {keys [<at 0.1 0>,<at 1 1>]}}}
        CurveIndex=&Abs_Float(Float:NiagaraFloat=&DotProduct(Vector A=<0 0 1>, Vector B=&ReturnCameraVector(Camera Vectors=::NewEnumerator1, Coordinate Space=::Simulation)))
       )
      )
      Y=1
      Z=1
     )
    ))
   /SetVariables(Particles.DynSmoke:Vector4f=&MakeVector4(
     X=&FloatFromCurve(FloatCurve=DataInterface() {propertyValues {Curve {keys [{arriveTangent=-2.6331, interpMode=RCIM_Cubic, leaveTangent=-2.6331, tangentMode=RCTM_User, time=0, value=0.6}, <at 1 -0.5 RCIM_Cubic>]}}})
     Y=&FloatFromCurve(FloatCurve=DataInterface() {propertyValues {Curve {keys [{arriveTangent=-0.5818, interpMode=RCIM_Cubic, leaveTangent=-0.5818, tangentMode=RCTM_User, time=0, value=1}, <at 1 0.9>]}}})
     Z=&FloatFromCurve(FloatCurve=DataInterface() {propertyValues {Curve {keys [<at 0 0>,<at 0.15 1>,<at 1 1.2>]}}})
     W=0
    )) {isSetParameters=true}
  ]
  rendererClasses [&NiagaraMeshRendererProperties]
  renderers [
   {rendererClass=&NiagaraMeshRendererProperties, rendererIndex=0}
   {rendererClass=&NiagaraMeshRendererProperties, rendererIndex=1}
  ]
 }
 MistSmoke {
  emitterUpdateScript [
   &EmitterState(Inactive Response=::NewEnumerator1, Life Cycle Mode=::NewEnumerator1, Loop Behavior=::NewEnumerator1, Loop Duration=4, UseLoopDelay=false)
   /SetVariables(Emitter.Shared.Curve_01=DataInterface() {propertyValues {Curve {
       defaultValue=1
       keys [<at 0 0 RCIM_Cubic>, {arriveTangent=0.6667, interpMode=RCIM_Cubic, leaveTangent=0.6667, time=1, value=0.85}, <at 1.5 1 RCIM_Cubic>]
      }}}) {isSetParameters=true}
   &SpawnBurst_Instantaneous(Use Spawn Probability=false, UseLoopCountLimit=false)
   &SpawnBurst_Instantaneous(Spawn Group=1, Use Spawn Probability=false, UseLoopCountLimit=false) {module=SpawnBurst_Instantaneous001}
   /SetVariables(Emitter.SecNoise_Control:Vector4f=&MakeVector4(
     X=-0.5
     Y=-0.5
     Z=0
     W=&Lerp_Float(
      A=-0.2
      B=-0.616
      Alpha=&FloatFromCurve(
       FloatCurve=DataInterface() {propertyValues {Curve {
          defaultValue=1
          keys [{arriveTangent=0.7798, interpMode=RCIM_Cubic, leaveTangent=0.7798, tangentMode=RCTM_User, time=0, value=0}, <at 1.5 1 RCIM_Cubic>]
         }}}
       CurveIndex=$Emitter.Age
      )
     )
    )) {isSetParameters=true, module=SetVariables_#2}
   /SetVariables(
    Emitter.MainNoise_Intensity:NiagaraFloat=&Lerp_Float(
     A=0
     B=0.2
     Alpha=&FloatFromCurve(
      FloatCurve=?
      Emitter.Shared.Curve_01:NiagaraDataInterfaceCurve=DataInterface() {propertyValues {Curve {
         defaultValue=1
         keys [<at 0 0 RCIM_Cubic>, {arriveTangent=0.6667, interpMode=RCIM_Cubic, leaveTangent=0.6667, time=1, value=0.85}, <at 1.5 1 RCIM_Cubic>]
        }}}
      CurveIndex=$Emitter.Age
     )
    )
    Emitter.SecNoise_Control_2:Vector4f=&MakeVector4(
     X=-0.152
     Y=-1
     Z=&Lerp_Float(
      A=0
      B=0.1
      Alpha=&FloatFromCurve(
       FloatCurve=DataInterface() {propertyValues {Curve {defaultValue=1, keys [<at 0 0 RCIM_Cubic>, {interpMode=RCIM_Cubic, leaveTangent=2.2145, time=1, value=1}]}}}
       CurveIndex=$Emitter.Age
      )
     )
     W=&Lerp_Float(
      A=0
      B=0.1
      Alpha=&FloatFromCurve(
       FloatCurve=DataInterface() {propertyValues {Curve {defaultValue=1, keys [<at 0 0 RCIM_Cubic>,<at 1 1 RCIM_Cubic>]}}}
       CurveIndex=$Emitter.Age
      )
     )
    )
    Emitter.ShapeOffset:Vector4f=&MakeVector4(
     X=&Lerp_Float(
      A=1
      B=1.8
      Alpha=&FloatFromCurve(
       FloatCurve=?
       Emitter.Shared.Curve_01:NiagaraDataInterfaceCurve=DataInterface() {propertyValues {Curve {
          defaultValue=1
          keys [<at 0 0 RCIM_Cubic>, {arriveTangent=0.6667, interpMode=RCIM_Cubic, leaveTangent=0.6667, time=1, value=0.85}, <at 1.5 1 RCIM_Cubic>]
         }}}
       CurveIndex=$Emitter.Age
      )
     )
     Y=1.8
     Z=0
     W=&Lerp_Float(
      A=-0.418
      B=0.05
      Alpha=&FloatFromCurve(
       FloatCurve=DataInterface() {propertyValues {Curve {
          defaultValue=1
          keys [
           {arriveTangent=-0.1091, interpMode=RCIM_Cubic, leaveTangent=-0.1091, tangentMode=RCTM_User, time=0, value=0}
           {arriveTangent=0.2114, interpMode=RCIM_Cubic, leaveTangent=0.2114, tangentMode=RCTM_User, time=0.3, value=0.2}
           {arriveTangent=0.5801, interpMode=RCIM_Cubic, leaveTangent=0.5801, tangentMode=RCTM_User, time=2, value=1}
          ]
         }}}
       CurveIndex=$Emitter.Age
      )
     )
    )
   ) {isSetParameters=true}
  ]
  particleSpawnScript [
   /SetVariables(Particles.VisibilityTag=$Emitter.SpawnGroup) {isSetParameters=true}
   &InitializeParticle(
    Lifetime=&Add_Float(A=$User.ExplosionDelay, B=0.2)
    Sprite Size=&MergeHelper_Vector2Switch(0=<300 250>, 1=<150 250>, 2=?, 3=?, 4=?, Index=$Particles.VisibilityTag, Num=::NewEnumerator0)
    Sprite Size Mode=::NewEnumerator4
    UsePositionOffset=false
    Write Material Random=true
   )
   &CameraOffset(Camera Offset Amount=&MergeHelper_FloatSwitch(0=100, 1=50, 2=?, 3=?, Index=$Particles.VisibilityTag, Num=::NewEnumerator0))
  ]
  particleUpdateScript [
   &ParticleState()
   &ScaleSpriteSize(
    Scale Factor=&MergeHelper_Vector2Switch(
     0=&Vector2DFromCurve(
      Vector2Curve=DataInterface() {propertyValues {
        XCurve {keys [{tangentMode=RCTM_User, time=0, value=0.2}, {tangentMode=RCTM_User, time=1, value=1}]}
        YCurve {keys [{tangentMode=RCTM_User, time=0, value=0.5}, {tangentMode=RCTM_User, time=1, value=1}]}
       }}
      ScaleCurve=<1 1>
     )
     1=&Vector2DFromCurve(
      Vector2Curve=DataInterface() {propertyValues {
        XCurve {keys [{tangentMode=RCTM_User, time=0, value=1}, {tangentMode=RCTM_User, time=1, value=1}]}
        YCurve {keys [{tangentMode=RCTM_User, time=0, value=0.8}, {tangentMode=RCTM_User, time=1, value=1}]}
       }}
      ScaleCurve=<1 1>
     )
     2=?
     3=?
     4=?
     Index=$Particles.Initial.VisibilityTag
     Num=::NewEnumerator0
    )
    Scale Sprite Size Mode=::NewEnumerator2
   )
   /SetVariables(Particles.SpriteAlignment=<0 0 1>) {isSetParameters=true, module=SetVariables_#3}
   &DynamicMaterialParameters(
    Index 0 Param 4=1
    Module.Index0Float0Write=true
    Module.Index0Float1Write=true
    Module.Index0Float2Write=true
    Module.Index0Float3Write=true
   )
   &ScaleColor(
    Scale Alpha=&FloatFromCurve(
     FloatCurve=DataInterface() {propertyValues {Curve {keys [<at 0 0>,<at 0.2 1>,<at 1 1>]}}}
     Scale Curve=&Lerp_Float(
      A=1
      B=0
      Alpha=&FloatFromCurve(
       FloatCurve=DataInterface() {propertyValues {Curve {keys [<at 0.35 0 RCIM_Cubic>,<at 0.9 1 RCIM_Cubic>]}}}
       CurveIndex=&Abs_Float(Float:NiagaraFloat=&DotProduct(Vector A=<0 0 1>, Vector B=&ReturnCameraVector(Camera Vectors=::NewEnumerator1, Coordinate Space=::Simulation)))
      )
     )
    )
    ScaleA=true
    ScaleRGB=true
   )
   &FN_Sprite_Attributes(
    Color=&MakeLinearColorFromVectorAndFloat(
     "Vector (RGB)"=<1 1 1>
     "Float (Alpha)"=&FloatFromCurve(
      FloatCurve=DataInterface() {propertyValues {Curve {keys [<at 0.2 0 RCIM_Cubic>, {arriveTangent=1.25, leaveTangent=1.25, time=0.6, value=1}, <at 1 1 RCIM_Cubic>]}}}
      Scale Curve=&Lerp_Float(
       A=0
       B=1
       Alpha=&FloatFromCurve(
        FloatCurve=DataInterface() {propertyValues {Curve {keys [<at 0.3 0 RCIM_Cubic>,<at 0.8 1 RCIM_Cubic>]}}}
        CurveIndex=&Abs_Float(Float:NiagaraFloat=&DotProduct(Vector A=<0 0 1>, Vector B=&ReturnCameraVector(Camera Vectors=::NewEnumerator1, Coordinate Space=::Simulation)))
       )
      )
     )
    )
    Create Color Attribute=true
    Create Position Attribute=true
    Create SpriteSize Attribute=true
    Position:NiagaraPosition=&AddVectorToPosition(Position=/Niagara/DynamicInputs/Helpers/SimulationPosition(), Vector=<0 0 100>)
    SpriteSize=<200 100>
    Write to Module Namespace=true
   )
  ]
  rendererClasses [&NiagaraSpriteRendererProperties]
  renderers [
   {rendererClass=&NiagaraSpriteRendererProperties}
   {rendererClass=&NiagaraSpriteRendererProperties, rendererIndex=1}
   {rendererClass=&NiagaraSpriteRendererProperties, rendererIndex=2}
  ]
 }
 Pull_In_Wave {
  emitterUpdateScript [
   &EmitterState(
    Inactive Response=::NewEnumerator1
    Life Cycle Mode=::NewEnumerator1
    Loop Behavior=::NewEnumerator1
    Loop Delay=&Subtract_Float(A=$User.ExplosionDelay, B=1.37)
    UseLoopDelay=true
   )
   &SpawnBurst_Instantaneous(Spawn Count=1, Use Spawn Probability=false)
  ]
  particleSpawnScript [
   &InitializeParticle(
    Lifetime=1.5
    Mesh Scale=<3 4.5 1>
    Mesh Scale Mode=::NewEnumerator4
    Position Offset=<0 0 100>
    UsePositionOffset=true
    Write Material Random=true
   )
   &InitialMeshOrientation(Orientation Axis=<0 0 1>, Orientation Vector=<0 0 1>, Use Orientation Vector=true, Use Rotation Vector=true) {enabled=false}
   &CameraOffset(Camera Offset Amount=30)
  ]
  particleUpdateScript [
   &ParticleState()
   &DynamicMaterialParameters(
    Index 0 Param 1=&FloatFromCurve(FloatCurve=DataInterface() {propertyValues {Curve {keys [
         {arriveTangent=3.1898, interpMode=RCIM_Cubic, time=0, value=-0.6}
         {arriveTangent=1.617, interpMode=RCIM_Cubic, leaveTangent=1.617, tangentMode=RCTM_User, time=1, value=0.7239}
        ]}}})
    Index 0 Param 2=1
    Index 0 Param 3=-1
    Module.Param0WriteEnabled=true
    Module.Param1WriteEnabled=true
    Module.Param2WriteEnabled=true
    Module.Param3WriteEnabled=false
   ) {module=DynamicMaterialParameters001}
   &ScaleColor(
    Linear Color Curve=DataInterface() {propertyValues {
      AlphaCurve {keys [<at 0.01 0>,<at 0.1374 1>,<at 0.4782 1>,<at 1 0>]}
      BlueCurve {keys [<at 0.0376 0.0156>]}
      GreenCurve {keys [<at 0.0376 0.0055>]}
      RedCurve {keys [<at 0.0376 0.0027>]}
     }}
    Scale Mode=::NewEnumerator2
   )
  ]
  rendererClasses [&NiagaraMeshRendererProperties]
  renderers [{rendererClass=&NiagaraMeshRendererProperties, rendererIndex=0}]
 }
 Sphere_Wind_LWE {rendererClasses [], renderers []}
}
summary {
 systemName=NS_Emote_BirchPalace_Intro
 emitters [
  {emitterName=MistSmoke, rendererClasses [&NiagaraSpriteRendererProperties]}
  {emitterName=ExpansionWave, rendererClasses [&NiagaraMeshRendererProperties]}
  {emitterName=Pull_In_Wave, rendererClasses [&NiagaraMeshRendererProperties]}
  {emitterName=Airflow, rendererClasses [&NiagaraMeshRendererProperties]}
  {emitterName=Chromatic, rendererClasses [&NiagaraSpriteRendererProperties]}
  {emitterName=Detail_Smoke, rendererClasses [&NiagaraSpriteRendererProperties]}
  {emitterName=Sphere_Wind_LWE, rendererClasses []}
 ]
 userVariables [
  User.bIsFrontEnd {type=NiagaraBool}
  User.bIsFrontEndPreview {type=NiagaraBool}
  User.ExplosionDelay {type=NiagaraFloat, defaultValue=3.6, description=""}
 ]
}
